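# Sound cue for the Badass Cactus Cowboy pickaxe swing.
# A mixer layers two attenuated branches: close uses the pickaxe's own swing recordings, medium uses shared distant fire axe swings.
# Each branch runs through a modulator and a quality-level switch that steps from larger random pools down to a single wave.
# It primes on load and overrides concurrency with a small per-owner limit, stopping the oldest.
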SoundCue /BRCosmetics/Sounds/PickAxes/BadassCactusCowboy/PA_BadassCactus_Swing_Cue
root=Mixer_3
nodes [
 {id=Attenuation_6, children [Modulator_7], properties {attenuationSettings=/Game/Sounds/Attenuation/Fort_Weapon_Pickaxe_Close_Att}}
 {id=Attenuation_7, children [Modulator_4], properties {attenuationSettings=/Game/Sounds/Attenuation/Fort_Weapon_Pickaxe_Medium_Att}}
 {id=Mixer_3, children [Attenuation_6, Attenuation_7], properties {inputVolume [1,1]}}
 {id=Modulator_4, children [QualityLevel_2]}
 {id=Modulator_7, children [QualityLevel_1]}
 {id=QualityLevel_1, children [Random_7, Random_9, WavePlayer_48, WavePlayer_48]}
 {id=QualityLevel_2, children [Random_1,Random_4,WavePlayer_1,WavePlayer_1]}
 {id=Random_1, children [WavePlayer_0,WavePlayer_3,WavePlayer_6,WavePlayer_7,WavePlayer_8,WavePlayer_9], properties {weights [1,1,1,1,1,1]}}
 {id=Random_4, children [WavePlayer_11, WavePlayer_10, WavePlayer_14], properties {weights [1,1,1]}}
 {id=Random_7, children [WavePlayer_30, WavePlayer_31, WavePlayer_38, WavePlayer_44], properties {weights [1,1,1,1]}}
 {id=Random_9, children [WavePlayer_45, WavePlayer_46, WavePlayer_47], properties {weights [1,1,1]}}
 {id=WavePlayer_0, properties {soundWaveAssetPtr=/Game/Athena/Sounds/Weapons/Fire_Axe_Swing_Distant_04}}
 {id=WavePlayer_1, properties {soundWaveAssetPtr=/Game/Athena/Sounds/Weapons/Fire_Axe_Swing_Distant_06}}
 {id=WavePlayer_10, properties {soundWaveAssetPtr=/Game/Athena/Sounds/Weapons/Fire_Axe_Swing_Distant_03}}
 {id=WavePlayer_11, properties {soundWaveAssetPtr=&Fire_Axe_Swing_Distant_04}}
 {id=WavePlayer_14, properties {soundWaveAssetPtr=&Fire_Axe_Swing_Distant_06}}
 {id=WavePlayer_3, properties {soundWaveAssetPtr=/Game/Athena/Sounds/Weapons/Fire_Axe_Swing_Distant_01}}
 {id=WavePlayer_30, properties {soundWaveAssetPtr=&/PA_BadassCactusCowboy_Swing_01}}
 {id=WavePlayer_31, properties {soundWaveAssetPtr=&/PA_BadassCactusCowboy_Swing_02}}
 {id=WavePlayer_38, properties {soundWaveAssetPtr=&/PA_BadassCactusCowboy_Swing_03}}
 {id=WavePlayer_44, properties {soundWaveAssetPtr=&/PA_BadassCactusCowboy_Swing_04}}
 {id=WavePlayer_45, properties {soundWaveAssetPtr=&PA_BadassCactusCowboy_Swing_01}}
 {id=WavePlayer_46, properties {soundWaveAssetPtr=&PA_BadassCactusCowboy_Swing_02}}
 {id=WavePlayer_47, properties {soundWaveAssetPtr=&PA_BadassCactusCowboy_Swing_03}}
 {id=WavePlayer_48, properties {soundWaveAssetPtr=&PA_BadassCactusCowboy_Swing_01}}
 {id=WavePlayer_6, properties {soundWaveAssetPtr=/Game/Athena/Sounds/Weapons/Fire_Axe_Swing_Distant_02}}
 {id=WavePlayer_7, properties {soundWaveAssetPtr=&Fire_Axe_Swing_Distant_03}}
 {id=WavePlayer_8, properties {soundWaveAssetPtr=/Game/Athena/Sounds/Weapons/Fire_Axe_Swing_Distant_05}}
 {id=WavePlayer_9, properties {soundWaveAssetPtr=&Fire_Axe_Swing_Distant_06}}
]
properties {
 audioPropertiesSheet=/BRCosmetics/Sounds/Settings/PropertySheets/Pickaxe/APS_Cosmetic_Pickaxe_Swing
 bOverrideConcurrency=true
 bPrimeOnLoad=true
 maxDistance=3000
 soundClassObject=&Fort_WeaponMelee_PSM
 volumeMultiplier=0.3
 concurrencyOverrides {bLimitToOwner=true, maxCount=2, resolutionRule=StopOldest}
 concurrencySet [&Player_Melee_Weapon_Conc]
}
